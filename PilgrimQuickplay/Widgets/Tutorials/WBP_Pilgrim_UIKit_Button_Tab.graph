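# Pilgrim Quickplay tab button widget, a child of the UIKit tab button that adjusts background corners and selection transitions.
# PreConstruct casts the block background to side-nav or top-nav type and rounds its corners more when IsPilgrimInverse is true.
# Selection uses a one-shot instant transition when IsInstantSelection is set.
# Otherwise, selection and deselection pick focused visual states when EnableFocusedSelectedState is on and the button is hovered or the game is mobile.
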WidgetBlueprint /PilgrimQuickplay/Widgets/Tutorials/WBP_Pilgrim_UIKit_Button_Tab
graphs [EventGraph {nodes [
   n1 = Event PreConstruct()
   > Parent: PreConstruct(IsDesignTime=@n1.IsDesignTime)
   > n2 = Branch(Condition=IsPilgrimInverse!(OwningPlayerController:Player Controller Object Reference=GetPlayerController!(PlayerIndex:int=0) -> (ReturnValue:Player Controller Object Reference)) -> (IsPilgrimInverse:bool))
   n3 = Get Block_Background() -> (Block_Background:WBP UIKit Block Background Base Object Reference)
   > n4 = Cast To WBP_UIKit_Block_BackgroundTabSideNav!(execute, Object=@n3) -> (then, CastFailed, AsWBP UIKit Block Background Tab Side Nav:WBP UIKit Block Background Tab Side Nav Object Reference) {is=title K2Node_DynamicCast}
   > SetCornerRadiuses!(execute, self:WBP UIKit Block Background Tab Side Nav Object Reference=@n4, CornerRadius:float=30, RadiusInPixels:bool=true) -> (then)
   n5 = Cast To WBP_UIKit_Block_BackgroundTopNavBar!(execute=@n4.CastFailed, Object=@n3) -> (then, CastFailed, AsWBP UIKit Block Background Top Nav Bar:WBP UIKit Block Background Top Nav Bar Object Reference) {is=title K2Node_DynamicCast}
   > SetCornerRadiuses(self:WBP UIKit Block Background Top Nav Bar Object Reference=@n5, CornerRadius=30, RadiusInPixels=true)
   n6 = Cast To WBP_UIKit_Block_BackgroundTabSideNav(execute=@n2.else, Object=@n3)
   > SetCornerRadiuses(self:WBP UIKit Block Background Tab Side Nav Object Reference=@n6, CornerRadius=12, RadiusInPixels=true)
   n7 = Cast To WBP_UIKit_Block_BackgroundTopNavBar(execute=@n6.CastFailed, Object=@n3)
   > SetCornerRadiuses(self:WBP UIKit Block Background Top Nav Bar Object Reference=@n7, CornerRadius=12, RadiusInPixels=true)
   Event Construct()
   > Parent: Construct()
   Event On Selected()
   > SetBlocksIsSelected!(execute, self:WBP UIKit Button Generic Base Object Reference, IsSelected:bool=true) -> (then)
   > n8 = Branch(Condition=Get IsInstantSelection() -> (IsInstantSelection:bool))
   > RequestInstantTransitionOnBlocks!(execute, self:UIKit Modular Button Object Reference, State:EUIKitBlockInstantTransitionState Enum=InstantSelected) -> (then)
   > Set IsInstantSelection(IsInstantSelection:bool=false) -> (Output_Get:bool)
   n9 = OR Boolean(A=IsHovered(), B=IsMobileGame())
   n10 = Get EnableFocusedSelectedState() -> (EnableFocusedSelectedState:bool)
   RequestTransitionOnBlocks!(
    execute=@n8.else
    self:UIKit Modular Button Object Reference
    State:EUIKitBlockVisualState Enum=Select(Option 0:EUIKitBlockVisualState Enum=Selected, Option 1:EUIKitBlockVisualState Enum=Focused, Index=AND Boolean(A=@n10, B=@n9)) -> (ReturnValue:EUIKitBlockVisualState Enum)
   ) -> (then)
   Event On Deselected()
   > SetBlocksIsSelected(IsSelected=false)
   n11 = OR Boolean(A=IsHovered(), B=IsMobileGame())
   n12 = Get EnableFocusedSelectedState()
   > RequestTransitionOnBlocks(State=Select(Option 0=DeselectedIdle, Option 1=DeselectedFocused, Index=AND Boolean(A=@n12, B=@n11)))
   Event Tick()
  ]}]
variables []
widgets {
 info {
  inheritedWidgetCount=0
  parentClass=/FortUILibrary/UIKit/Buttons/TabButtons/WBP_UIKit_Button_Tab.WBP_UIKit_Button_Tab_C
  rootWidgetClass=null
  widgetCount=0
 }
 widgets []
}
