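# Tab list widget for Pilgrim Quickplay, derived from the shared UIKit tabs base and adding no widgets of its own.
# Its SetupTabButton override checks IsPilgrimInverse for the first player controller to choose a larger or zero MinButtonHeight.
# It stores that height before calling the parent setup with the same tab ID and button.
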WidgetBlueprint /PilgrimQuickplay/Widgets/WBP_UIKit_Tabs_Pilgrim
graphs [
 EventGraph {links [], nodes []}
 SetupTabButton {nodes [
   n1 = SetupTabButton!() -> (then, Tab ID:Name=None, CreatedButtonReference:Common Button Base Object Reference) {is=title K2Node_FunctionEntry}
   n2 = GetPlayerController!(PlayerIndex:int=0) -> (ReturnValue:Player Controller Object Reference)
   > Set MinButtonHeight(MinButtonHeight:int=SelectInt(A=100, B=0, bPickA=IsPilgrimInverse!(OwningPlayerController:Player Controller Object Reference=@n2) -> (IsPilgrimInverse:bool))) -> (Output_Get:int=0)
   > Parent: SetupTabButton!(execute, Tab ID:Name=@n1, CreatedButtonReference:Common Button Base Object Reference=@n1.CreatedButtonReference) -> (then) {is=title K2Node_CallParentFunction}
  ]}
]
variables [Min Button Height {type=integer}]
widgets {
 info {
  inheritedWidgetCount=0
  parentClass=/FortUILibrary/Systems/Tabs/WBP_UIKit_Tabs_Base.WBP_UIKit_Tabs_Base_C
  rootWidgetClass=null
  widgetCount=0
 }
 widgets []
}
