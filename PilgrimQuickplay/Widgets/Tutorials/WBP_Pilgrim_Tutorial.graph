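# Full-screen Main Stage tutorial modal for Pilgrim Quickplay, showing info panels beside vertical tabs.
# On init it filters panel data assets by platform, then builds the panels and view-model-driven tabs from them.
# Tab selection, mouse wheel and scroll events update a shared index that scrolls to and highlights the panel, with directional sounds.
# Close buttons, or a valid quickplay state-began event from the event router, remove the tutorial scene.
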WidgetBlueprint /PilgrimQuickplay/Widgets/Tutorials/WBP_Pilgrim_Tutorial
graphs [
 EventGraph {nodes [
   n1 = "On Clicked (Button_Close)"!() -> (OutputDelegate, then, Button:Common Button Base Object Reference) {is=title K2Node_ComponentBoundEvent}
   RemoveTipsScene(execute=[@n1.then, @n2.then, @n3.then])
   n2 = "On Button Clicked (Button_CloseTouch)"!() -> (OutputDelegate, then, Button:Common Button Legacy Object Reference) {is=title K2Node_ComponentBoundEvent}
   n3 = "On Clicked (Button_Close_Alt)"!() -> (OutputDelegate, then, Button:Common Button Base Object Reference) {is=title K2Node_ComponentBoundEvent}
   n4 = "On Pressed (Button_Close)"!() -> (OutputDelegate, then, Button:Common Button Base Object Reference) {is=title K2Node_ComponentBoundEvent}
   PlaySound2D(
    execute=[@n4.then, @n5.then]
    Sound=/Game/Sounds/Fort_UI_Sounds/FrontEnd/DefaultSounds/PressedButtons/MSS_UI_Button_Pressed_Default
    VolumeMultiplier=0.7
   )
   n5 = "On Pressed (Button_Close_Alt)"!() -> (OutputDelegate, then, Button:Common Button Base Object Reference) {is=title K2Node_ComponentBoundEvent}
   "On Clicked (Button_CloseTouch)"!() -> (OutputDelegate, then, Button:Common Button Base Object Reference) {is=title K2Node_ComponentBoundEvent}
   > PlaySound2D(Sound=/Game/Sounds/Fort_UI_Sounds/FrontEnd/New_FrontEnd/Ranked/Fortnite_UI_Ranked_Button_Back_Cue, VolumeMultiplier=0.7)
   n6 = OnTabSelected!() -> (OutputDelegate, then, TabId:Name) {is=title K2Node_CustomEvent}
   > n7 = Set CurrentIndex(CurrentIndex:int=GetTabIndexByID!(
     self:UIKit Tab Group Object Reference=Get WBP_UIKit_Tabs_Pilgrim() -> (WBP_UIKit_Tabs_Pilgrim:WBP UIKit Tabs Pilgrim Object Reference)
     TabNameID:Name=@n6.TabId
    ) -> (ReturnValue:int=0)) -> (Output_Get:int=0)
   > UpdateFocusedPanel!(execute, self, Index:int=@n7) -> (then)
   RemoveTipsScene() {is=title K2Node_CustomEvent}
   > RemoveScene!(
    execute
    self:Dynamic UIManager Object Reference=Get DynamicUIManager() -> (ReturnValue:Dynamic UIManager Object Reference) {is=title K2Node_GetSubsystem}
    Scene:Dynamic UIScene Object Reference=Get SceneToRemoveOnClose() -> (SceneToRemoveOnClose:Dynamic UIScene Object Reference)
    Player:"Player Controller Object Reference (by ref)"=GetOwningPlayer()
   ) -> (then)
   n8 = Event Construct()
   SetPadding!(
    execute=[@n8.then, @n12.then]
    self:Border Slot Object Reference=SlotAsBorderSlot!(Widget:Widget Object Reference=Get SafeZone_content() -> (SafeZone_content:Safe Zone Object Reference)) -> (ReturnValue:Border Slot Object Reference)
    InPadding_Left:float=64
    InPadding_Top:float=SelectFloat(A=32, B=64, bPickA=IsMobileGame())
    InPadding_Right:float=64
    InPadding_Bottom:float=32
   ) -> (then)
   > "SetText (Text)"(self=Get Text_HeaderMain() -> (Text_HeaderMain:Common Text Object Reference), InText=Get MainHeaderText() -> (MainHeaderText:Text))
   > "SetText (Text)"(self=Get Text_HeaderSub() -> (Text_HeaderSub:Common Text Object Reference), InText=Get SubHeaderText() -> (SubHeaderText:Text))
   Event OnInitialized()
   > SanitizeDataInfoForPlatform!(execute, self) -> (then, SanitizedDataInfo:Array of BP UI Pilgrim Info Panel Object References)
   > Populate Panels()
   > Populate Buttons()
   > UpdateFocusedPanel(Index=Get CurrentIndex() -> (CurrentIndex:int=0))
   > n9 = PilgrimWaitForEventRouter!(execute:Exec, ContextActor:Actor Object Reference=GetOwningPlayer()) -> (then:Exec, OnEventRouterAvailable:Exec, EventRouter:Gameplay Event Router Component Object Reference) {is=title K2Node_AsyncAction}
   n10 = StartListeningToEvent!(
    execute:Exec=@n9.OnEventRouterAvailable
    Target:Gameplay Event Router Component Object Reference=@n9
    EventType:Script Struct Object Reference=/Script/PilgrimQPSharedRuntime.PilgrimEvent_QuickplayState_Began
    Context:Object Reference
    EventBubblingRule:EEventBubblingRule Enum
   ) -> (
    then:Exec
    AsyncTaskProxy:Async Action Start Listening To Event Object Reference
    OnEventReceived:Exec
    Payload:Pilgrim Event Quickplay State Began Structure
   )
   Branch(
    execute=@n10.OnEventReceived
    Condition="Not Equal (Enum)"!(
     A:EPilgrimQuickplayState Enum=Break Pilgrim Event Quickplay State Base!(PilgrimEvent_QuickplayState_Base:"Pilgrim Event Quickplay State Base Structure (by ref)"=@n10.Payload) -> (StateId:EPilgrimQuickplayState Enum) {is=title K2Node_BreakStruct}
     B:EPilgrimQuickplayState Enum=Invalid
    ) -> (ReturnValue:bool)
   )
   > RemoveTipsScene() {is=title K2Node_CallFunction}
   Event On Activated()
   > PlaySound2D(Sound=Get AppearSoundCue() -> (AppearSoundCue:Sound Base Object Reference))
   > DelayUntilNextTick!(execute) -> (then)
   n11 = Get WBP_UIKit_Tabs_Pilgrim()
   > SetFocus(self=GetChildAt!(
     self:Panel Widget Object Reference=Get TabsContainer!(self:WBP UIKit Tabs Base Object Reference=@n11) -> (TabsContainer:Grid Panel Object Reference)
     Index:int=0
    ) -> (ReturnValue:Widget Object Reference))
   > Branch(Condition=IsPilgrimInverse!(OwningPlayerController:Player Controller Object Reference=GetPlayerController!(PlayerIndex:int=0) -> (ReturnValue:Player Controller Object Reference)) -> (IsPilgrimInverse:bool))
   > Set MinButtonHeight!(
    execute
    MinButtonHeight:int=Get Mobile Min Button Height() -> (Mobile Min Button Height:int=0)
    self:WBP UIKit Tabs Base Object Reference=Get WBP_UIKit_Tabs_Pilgrim()
   ) -> (then, Output_Get:int=0)
   > SetPadding(self=SlotAsBorderSlot(Widget=Get SafeZone_content()), InPadding_Left=64, InPadding_Top=32, InPadding_Right=64, InPadding_Bottom=32)
   n12 = Event PreConstruct()
   n13 = Event On Scroll Completed!() -> (OutputDelegate, then, bScrolledDown:bool) {is=title K2Node_Event}
   > EndInertialScrolling(self:Scroll Box Object Reference=Get ScrollBox_Panels() -> (ScrollBox_Panels:Athena Scroll Box Object Reference))
   > n14 = Branch(Condition=@n13.bScrolledDown)
   > ScrolledDown()
   ScrolledUp(execute=@n14.else)
   Mouse Wheel Down!() -> (Pressed:Exec, Released:Exec, Key:Key Structure) {is=title K2Node_InputKey, inputKey=MouseScrollDown}
   > ScrolledDown()
   Mouse Wheel Up!() -> (Pressed:Exec, Released:Exec, Key:Key Structure) {is=title K2Node_InputKey, inputKey=MouseScrollUp}
   > ScrolledUp()
  ]}
 Populate Buttons {nodes [
   Populate Buttons() {is=title K2Node_FunctionEntry}
   > n1 = Branch(Condition=Is Not Empty!(TargetArray:Array of BP UI Pilgrim Info Panel Object References=Get SanitizedDataInfo() -> (SanitizedDataInfo:Array of BP UI Pilgrim Info Panel Object References)) -> (ReturnValue:bool))
   > n2 = Construct Object from Class(Class=/Script/FortniteUI.UIKitTabGroupViewModel) -> (ReturnValue:UI Kit Tab Group View Model Object Reference)
   > n3 = For Each Loop(Array:Array of BP UI Pilgrim Info Panel Object References=Get SanitizedDataInfo()) -> (Array Element:BP UI Pilgrim Info Panel Object Reference)
   Set Tabs!(
    execute=@n3.Completed
    Tabs:Array of UI Kit Tab Group Button View Model Object References=Get TabsDataInfo() -> (TabsDataInfo:Array of UI Kit Tab Group Button View Model Object References)
    self:UI Kit Tab Group View Model Object Reference=@n2
   ) -> (then, Output_Get:Array of UI Kit Tab Group Button View Model Object References)
   n4 = Get WBP_UIKit_Tabs_Pilgrim()
   > Set UIKitTabGroupViewModel!(execute, UIKitTabGroupViewModel:UI Kit Tab Group View Model Object Reference=@n2, self:WBP UIKit Tabs Pilgrim Object Reference=@n4) -> (then, Output_Get:UI Kit Tab Group View Model Object Reference)
   > Bind Event to On Tab Selected(self:Common Tab List Widget Base Object Reference=@n4, Delegate=Create Event()) {is=title K2Node_AddDelegate}
   n5 = Construct Object from Class(execute=@n3.LoopBody, Class=/Script/FortniteUI.UIKitTabGroupButtonViewModel) -> (ReturnValue:UI Kit Tab Group Button View Model Object Reference)
   > Set Label!(
    execute
    Label:Text=Get ButtonName!(self:BP UI Pilgrim Info Panel Object Reference=@n3) -> (ButtonName:Text)
    self:UI Kit Selectable Button View Model Object Reference=@n5
   ) -> (then, Output_Get:Text)
   > Add(
    TargetArray:Array of UI Kit Tab Group Button View Model Object References=Get TabsDataInfo()
    NewItem:"UI Kit Tab Group Button View Model Object Reference (by ref)"=@n5
   ) {is=title K2Node_CallArrayFunction}
   FortShippingLog!(
    execute=@n1.else
    WorldContextObject:Object Reference
    InString:String=Attempting to PopulateButtons without sanitized data info
    bLogAsWarning:bool=false
   ) -> (then)
  ]}
 Populate Panels {nodes [
   Populate Panels() {is=title K2Node_FunctionEntry}
   > n1 = Branch(Condition=Is Not Empty(TargetArray=Get SanitizedDataInfo()))
   > n2 = For Each Loop(Array=Get SanitizedDataInfo())
   > n3 = Create Widget!(
    execute:Exec
    Class:User Widget Class Reference=Get Panel Type!(self:BP UI Pilgrim Info Panel Object Reference=@n2) -> (Panel Type:User Widget Class Reference)
    OwningPlayer:Player Controller Object Reference
   ) -> (then:Exec, ReturnValue:User Widget Object Reference)
   > AddChildToVerticalBox(self=Get Vbox_Panels() -> (Vbox_Panels:Vertical Box Object Reference), Content=@n3)
   > UpdatePanel!(execute:Exec, self:Object Reference=@n3, Object In:Object Reference=@n2) -> (then:Exec) {is=title K2Node_Message}
   > UpdateFocus!(
    execute:Exec
    self:Object Reference=@n3
    IsFocused:bool=Select(Option 0:bool=false, Option 1:bool=true, Index="Equal (Integer)"(A=@n2.Array Index)) -> (ReturnValue:bool)
   ) -> (then:Exec) {is=title K2Node_Message}
   > UpdateEdgePadding!(execute, self, Index:int=@n2.Array Index, Object:Object Reference=@n3) -> (then)
   FortShippingLog(execute=@n1.else, InString=Attempting to PopulatePanels without sanitized data info, bLogAsWarning=false)
  ]}
 SanitizeDataInfoForPlatform {nodes [
   SanitizeDataInfoForPlatform() {is=title K2Node_FunctionEntry}
   > Clear(TargetArray:Array of BP UI Pilgrim Info Panel Object References=Get SanitizedDataInfo())
   > n1 = For Each Loop(Array=Get DataInfo() -> (DataInfo:Array of BP UI Pilgrim Info Panel Object References))
   Return Node!(execute=@n1.Completed, SanitizedDataInfo:Array of BP UI Pilgrim Info Panel Object References=Get SanitizedDataInfo())
   n2 = Get ExclusivePlatforms!(self:BP UI Pilgrim Info Panel Object Reference=@n1) -> (ExclusivePlatforms:Array of Strings)
   > n3 = Branch(Condition=Is Empty!(TargetArray=@n2) -> (ReturnValue:bool))
   > Add(
    TargetArray:Array of BP UI Pilgrim Info Panel Object References=Get SanitizedDataInfo()
    NewItem:"BP UI Pilgrim Info Panel Object Reference (by ref)"=@n1
   )
   n4 = GetPlatformName!() -> (ReturnValue:String)
   Branch(execute=@n3.else, Condition="integer >= integer"(A=Find Item!(TargetArray=@n2, ItemToFind:"String (by ref)"=@n4) -> (ReturnValue)))
   > Add(TargetArray=Get SanitizedDataInfo(), NewItem=@n1)
  ]}
 ScrolledDown {nodes [
   ScrolledDown() {is=title K2Node_FunctionEntry}
   > n1 = Branch(Condition=Is Not Empty(TargetArray=Get SanitizedDataInfo()))
   n2 = Get SanitizedDataInfo()
   n3 = Get CurrentIndex() -> (CurrentIndex=0)
   > n4 = Branch(Condition="integer < integer"(A=@n3, B=Last Index!(TargetArray:Array of BP UI Pilgrim Info Panel Object References=@n2) -> (ReturnValue=0)))
   > n5 = Increment Int(Value=@n3)
   > n6 = Set CurrentIndex(CurrentIndex=@n5) -> (Output_Get=0)
   > n7 = Set bScrollUp(bScrollUp:bool=false) -> (Output_Get:bool)
   UpdateFocusedPanel(execute=[@n7.then, @n4.else], Index=@n6)
   > SelectTabByIndex!(
    execute
    self:UIKit Tab Group Object Reference=Get WBP_UIKit_Tabs_Pilgrim()
    Index:int=Get CurrentIndex() -> (CurrentIndex=0)
    bSuppressClickFeedback:bool=true
   ) -> (then, ReturnValue:bool)
   FortShippingLog(execute=@n1.else, InString=Attempting to ScrollDown without sanitized data info, bLogAsWarning=false)
  ]}
 ScrolledUp {nodes [
   ScrolledUp() {is=title K2Node_FunctionEntry}
   n1 = Get CurrentIndex() -> (CurrentIndex=0)
   > n2 = Branch(Condition="Equal (Integer)"(A=@n1))
   n3 = Decrement Int(" "=@n2.else, Value=@n1)
   > n4 = Set CurrentIndex(CurrentIndex=@n3) -> (Output_Get=0)
   > Set bScrollUp(bScrollUp=true)
   > UpdateFocusedPanel(Index=@n4)
   > SelectTabByIndex(self=Get WBP_UIKit_Tabs_Pilgrim(), Index=Get CurrentIndex() -> (CurrentIndex=0), bSuppressClickFeedback=true)
  ]}
 UpdateEdgePadding {nodes [
   n1 = UpdateEdgePadding!() -> (then, Index:int, Object:Object Reference) {is=title K2Node_FunctionEntry}
   > n2 = Branch(Condition=Is Not Empty(TargetArray=Get SanitizedDataInfo()))
   > n3 = Cast To BPI_UI_Pilgrim_Tutorial_Entries!(execute, Object=@n1.Object) -> (then, CastFailed, AsBPI UI Pilgrim Tutorial Entries:BPI UI Pilgrim Tutorial Entries Interface) {is=title K2Node_DynamicCast}
   n4 = Get SanitizedDataInfo()
   > UpdateEndListPadding!(
    execute
    self:BPI UI Pilgrim Tutorial Entries Interface=@n3
    IsFirstEntry:bool="Equal (Integer)"(A=@n1)
    IsLastEntry:bool="Equal (Integer)"(A=@n1, B=Last Index(TargetArray=@n4) -> (ReturnValue=0))
   ) -> (then)
   FortShippingLog(execute=@n2.else, InString=Attempting to UpdateEdgePadding without sanitized data info, bLogAsWarning=false)
  ]}
 UpdateFocusedPanel {nodes [
   n1 = UpdateFocusedPanel!() -> (then, Index:int) {is=title K2Node_FunctionEntry}
   > ScrollWidgetIntoView!(
    execute
    self:Scroll Box Object Reference=Get ScrollBox_Panels()
    WidgetToFind:Widget Object Reference=GetChildAt(self=Get Vbox_Panels(), Index=@n1)
    AnimateScroll:bool=true
    ScrollDestination:EDescendantScrollDestination Enum=TopOrLeft
    Padding:float=256
   ) -> (then)
   > Branch(Condition="Not Equal (Integer)"(A=Get Index() -> (Index:int=0), B=Get LastIndex() -> (LastIndex:int=0)))
   > n2 = Cast To BPI_UI_Pilgrim_Tutorial_Entries(Object=GetChildAt(self=Get Vbox_Panels(), Index=Get Index() -> (Index=0)))
   > UpdateFocus!(execute, self:BPI UI Pilgrim Tutorial Entries Interface=@n2, IsFocused:bool=true) -> (then) {is=title K2Node_CallFunction}
   > n3 = Cast To BPI_UI_Pilgrim_Tutorial_Entries(Object=GetChildAt(self=Get Vbox_Panels(), Index=Get LastIndex() -> (LastIndex=0)))
   > UpdateFocus(self=@n3, IsFocused=false)
   > Set LastIndex(LastIndex:int=Get Index() -> (Index=0)) -> (Output_Get:int=0)
   > Branch(Condition=IsVisible())
   > n4 = Branch(Condition=Get bScrollUp() -> (bScrollUp:bool))
   > PlaySound2D(Sound=Get ScrollUpSoundCue() -> (ScrollUpSoundCue:Sound Base Object Reference))
   PlaySound2D(execute=@n4.else, Sound=Get ScrollDownSoundCue() -> (ScrollDownSoundCue:Sound Base Object Reference))
  ]}
]
variables [
 AppearSoundCue {type=SoundBase, default=/Game/Sounds/Fort_UI_Sounds/FrontEnd/New_FrontEnd/Ranked/Fort_UI_Ranked_Whoosh_Intro_Cue}
 bScrollUp {type=boolean}
 CurrentIndex {type=integer}
 DataInfo {
  type="array<BP_UI_Pilgrim_InfoPanel_C>"
  default [
   &/PilgrimQuickplay_TutorialCalibrationData
   &/PilgrimQuickplay_TutorialStartingAShowData
   &/PilgrimQuickplay_TutorialBasicsOnStageData
   &/PilgrimQuickplay_TutorialSpecialNotesData
   &/PilgrimQuickplay_TutorialOverdriveData
   &/PilgrimQuickplay_TutorialProVocals
   &/PilgrimQuickplay_TutorialBandPlayData
   &/PilgrimQuickplay_TutorialLocalPlayData
   &/PilgrimQuickplay_TutorialCustomizeBandData
  ]
 }
 In Padding {type=Margin, default=<ltrb 0 0 0 0>}
 LastIndex {type=integer, default=1}
 MainHeaderText {type=string, default=MAIN STAGE}
 Mobile Min Button Height {type=integer, default=100}
 Pins {type=Object, default=null}
 SanitizedDataInfo {
  type="array<BP_UI_Pilgrim_InfoPanel_C>"
  default [
   &PilgrimQuickplay_TutorialCalibrationData
   &PilgrimQuickplay_TutorialStartingAShowData
   &PilgrimQuickplay_TutorialBasicsOnStageData
   &PilgrimQuickplay_TutorialSpecialNotesData
   &PilgrimQuickplay_TutorialOverdriveData
   &PilgrimQuickplay_TutorialBandPlayData
  ]
 }
 SceneToRemoveOnClose {type=DynamicUIScene, default=/PilgrimQuickplay/Widgets/Scenes/UIScene_PilgrimQuickplay_Tutorials}
 ScrollDownSoundCue {type=SoundBase, default=/Game/Sounds/Fort_UI_Sounds/FrontEnd/New_FrontEnd/Ranked/Fort_UI_Shop_Whoosh_Down_Cue}
 ScrollUpSoundCue {type=SoundBase, default=/Game/Sounds/Fort_UI_Sounds/FrontEnd/New_FrontEnd/Ranked/Fort_UI_Ranked_Whoosh_Up_Cue}
 SubHeaderText {type=string, default=MODE INFORMATION}
 TabsDataInfo {type="array<UIKitTabGroupButtonViewModel>", default []}
]
widgets {
 info {inheritedWidgetCount=5, parentClass=/Script/PilgrimQuickplayUI.PilgrimQuickplayInfoModal, rootWidgetClass=&Border, widgetCount=25}
 widgets [
  Border_0 {
   widgetClassPath=&Border
   widget=Border_0
   properties {
    background {
     bIsDynamicallyLoaded=false
     drawAs=Image
     imageType=FullColor
     mirroring=NoMirror
     resourceName=None
     resourceObject=/FMCalibration/Materials/MI_UI_FTUE_Sparks_Fullscreen_Background
     tiling=NoTile
     imageSize=<32 32>
     margin=<ltrb 0 0 0 0>
     outlineSettings {
      bUseBrushTransparency=false
      roundingType=HalfHeightRadius
      width=0
      color {colorUseRule=UseColor_Specified, specifiedColor=<rgba 0 0 0 0>}
      cornerRadii=<0 0 0 0>
     }
     tintColor {colorUseRule=UseColor_Specified, specifiedColor=<rgba 1 1 1 1>}
     uVRegion {bIsValid=false, max=<0 0>, min=<0 0>}
    }
    padding=<ltrb 0 0 0 0>
   }
  }
  Overlay_23 {widgetClassPath=&Overlay, parent=Border_0, widget=Overlay_23, slot {class=BorderSlot, properties {padding=<ltrb 0 0 0 0>}}}
  WBP_CaptureForPostBufferUpdate {
   widgetClassPath=&WBP_CaptureForPostBufferUpdate
   bIsVariable=true
   parent=Overlay_23
   slot {class=OverlaySlot, properties {horizontalAlignment=HAlign_Fill, verticalAlignment=VAlign_Fill}}
  }
  HorizontalBox_0 {
   widgetClassPath=&HorizontalBox
   parent=Overlay_23
   widget=HorizontalBox_0
   slot {class=OverlaySlot, properties {horizontalAlignment=HAlign_Fill, verticalAlignment=VAlign_Fill}}
  }
  Border_LeftContent {
   widgetClassPath=&CommonBorder
   parent=HorizontalBox_0
   widget=Border_LeftContent
   properties {
    background {
     bIsDynamicallyLoaded=false
     drawAs=Image
     imageType=NoImage
     mirroring=NoMirror
     resourceName=None
     resourceObject=/FMCore/Materials/MusicLibraryBG_Inst_DublicateFMCore
     tiling=NoTile
     imageSize=<64 64>
     margin=<ltrb 0 0 0 0>
     outlineSettings {
      bUseBrushTransparency=false
      roundingType=HalfHeightRadius
      width=0
      color {colorUseRule=UseColor_Specified, specifiedColor=<rgba 0 0 0 0>}
      cornerRadii=<0 0 0 0>
     }
     tintColor {colorUseRule=UseColor_Specified, specifiedColor=<rgba 1 1 1 1>}
     uVRegion {bIsValid=false, max=<0 0>, min=<0 0>}
    }
    padding=<ltrb 32 32 64 32>
   }
   slot {class=HorizontalBoxSlot, properties {size {sizeRule=Automatic, value=0.25}}}
  }
  SafeZone_content {
   widgetClassPath=&SafeZone
   bIsVariable=true
   parent=Border_LeftContent
   widget=SafeZone_content
   properties {padBottom=false, padRight=false, padTop=false}
   slot {class=BorderSlot, properties {padding=<ltrb 32 32 64 32>}}
  }
  LeftSide {
   widgetClassPath=&VerticalBox
   parent=SafeZone_content
   widget=LeftSide
   slot {class=SafeZoneSlot, properties {safeAreaScale=<ltrb 1.25 0 0 0>}}
  }
  Spacer_187 {widgetClassPath=&Spacer, parent=LeftSide, widget=Spacer_187, properties {size=<200 200>}, slot {class=VerticalBoxSlot}}
  WBP_UIKit_Tabs_Pilgrim {
   widgetClassPath=/PilgrimQuickplay/Widgets/WBP_UIKit_Tabs_Pilgrim
   bIsVariable=true
   parent=LeftSide
   widget=WBP_UIKit_Tabs_Pilgrim
   properties {
    bOverride_Cursor=true
    backplateCornerRadius=Circle
    buttonClass=&/WBP_Pilgrim_UIKit_Button_Tab.WBP_Pilgrim_UIKit_Button_Tab_C
    dividerSpacing=24
    hidePrompts=true
    inputActions=None
    isNaturalNavigation=true
    isVertical=true
    minButtonWidth=300
    nextPromptHorizontalAlignment=HAlign_Left
    selectionAlignment=Left
    tabGroupHorizontalAlignment=HAlign_Left
    tabSpacing=10
    useBackplate=false
    localTabsViewModels []
    nextPromptPadding=<ltrb 64 0 0 0>
   }
   slot {class=VerticalBoxSlot, properties {verticalAlignment=VAlign_Top, size {sizeRule=Fill, value=1}}}
  }
  Spacer_383 {widgetClassPath=&Spacer, parent=HorizontalBox_0, widget=Spacer_383, properties {size=<32 1>}, slot {class=HorizontalBoxSlot}}
  Overlay_2 {
   widgetClassPath=&Overlay
   parent=HorizontalBox_0
   widget=Overlay_2
   slot {class=HorizontalBoxSlot, properties {padding=<ltrb 0 0 64 0>, size {sizeRule=Fill, value=1}}}
  }
  ScrollBox_Panels {
   widgetClassPath=/Script/FortniteUI.AthenaScrollBox
   bInherited=true
   bIsVariable=true
   parent=Overlay_2
   widget=ScrollBox_Panels
   properties {
    allowOverscroll=false
    bAllowRightClickDragScrolling=false
    bAnimateWheelScrolling=false
    bEnableClickThrough=true
    consumeMouseWheel=Never
    scrollBarVisibility=Collapsed
    scrollWhenFocusChanges=AnimatedScroll
    visibility=SelfHitTestInvisible
   }
   slot {class=OverlaySlot, properties {horizontalAlignment=HAlign_Fill, verticalAlignment=VAlign_Fill}}
  }
  Vbox_Panels {
   widgetClassPath=&VerticalBox
   bInherited=true
   bIsVariable=true
   parent=ScrollBox_Panels
   widget=Vbox_Panels
   slot {class=ScrollBoxSlot}
  }
  SafeZone_0 {
   widgetClassPath=&SafeZone
   bIsVariable=true
   parent=Overlay_2
   widget=SafeZone_0
   properties {padBottom=false, padLeft=false, padRight=false}
   slot {class=OverlaySlot, properties {horizontalAlignment=HAlign_Fill, padding=<ltrb 64 72 92 0>}}
  }
  Text_Header {widgetClassPath=&VerticalBox, parent=SafeZone_0, widget=Text_Header, slot {class=SafeZoneSlot}}
  Text_HeaderSub {
   widgetClassPath=&CommonTextBlock
   bIsVariable=true
   parent=Text_Header
   widget=Text_HeaderSub
   properties {
    mobileFontSizeMultiplier=1.4
    text=MODE INFORMATION
    textTransformPolicy=ToUpper
    colorAndOpacity {colorUseRule=UseColor_Specified, specifiedColor=<rgba 0.0314 0.702 1 1>}
    font {
     bForceMonospaced=false
     bMaterialIsStencil=false
     fontMaterial=null
     fontObject=&Burbank
     letterSpacing=0
     monospacedWidth=1
     size=18
     skewAmount=0
     typefaceFontName=BurbankRegular-Black
     outlineSettings {
      bApplyOutlineToDropShadows=false
      bMiteredCorners=false
      bSeparateFillAlpha=false
      outlineMaterial=null
      outlineSize=0
      outlineColor=<rgba 0 0 0 1>
     }
    }
    margin=<ltrb 4 0 4 0>
    shadowOffset=<0 0>
   }
   slot {class=VerticalBoxSlot, properties {padding=<ltrb 6 0 0 0>}}
  }
  Text_HeaderMain {
   widgetClassPath=&CommonTextBlock
   bIsVariable=true
   parent=Text_Header
   widget=Text_HeaderMain
   properties {
    autoWrapText=true
    text=MAIN STAGE
    textOverflowPolicy=Ellipsis
    textTransformPolicy=ToUpper
    font {
     bForceMonospaced=false
     bMaterialIsStencil=false
     fontMaterial=null
     fontObject=&Burbank
     letterSpacing=0
     monospacedWidth=1
     size=40
     skewAmount=0
     typefaceFontName=BurbankRegular-Black
     outlineSettings {
      bApplyOutlineToDropShadows=false
      bMiteredCorners=false
      bSeparateFillAlpha=false
      outlineMaterial=null
      outlineSize=0
      outlineColor=<rgba 0 0 0 1>
     }
    }
    margin=<ltrb 4 0 4 0>
    shadowOffset=<0 0>
   }
   slot {class=VerticalBoxSlot}
  }
  SafeZone_bottombuttons {
   widgetClassPath=&SafeZone
   bIsVariable=true
   parent=Overlay_23
   widget=SafeZone_bottombuttons
   properties {padLeft=false, padTop=false}
   slot {class=OverlaySlot, properties {horizontalAlignment=HAlign_Right, verticalAlignment=VAlign_Bottom, padding=<ltrb 0 0 36 32>}}
  }
  VisibilityBorder_NonTouch {
   widgetClassPath=&CommonHardwareVisibilityBorder
   parent=SafeZone_bottombuttons
   widget=VisibilityBorder_NonTouch
   properties {visibilityQuery {
     autoDescription=" NONE( Input.Touch )"
     tokenStreamVersion=0
     userDescription=non-touch
     queryTokenStream [0,1,3,1,0]
     tagDictionary [{tagName=Input.Touch}]
    }}
   slot {class=SafeZoneSlot}
  }
  Overlay_3 {
   widgetClassPath=&Overlay
   parent=VisibilityBorder_NonTouch
   widget=Overlay_3
   slot {class=BorderSlot, properties {padding=<ltrb 0 0 0 0>}}
  }
  Button_Close_Alt {
   widgetClassPath=/FortUILibrary/UIKit/Buttons/WBP_UIKit_Button_Quiet
   bInherited=true
   bIsVariable=true
   parent=Overlay_3
   widget=Button_Close_Alt
   properties {
    bDisplayInputActionWhenNotInteractable=false
    bHideInputAction=true
    bIsFocusable=false
    renderOpacity=0
    text=Close
    hoveredSlateSoundOverride {resourceObject=/Game/Sounds/Fort_UI_Sounds/FrontEnd/New_FrontEnd/Ranked/Fortnite_UI_Ranked_Button_Hover_Cue}
    pressedSlateSoundOverride {resourceObject=&Fortnite_UI_Ranked_Button_Back_Cue}
    selectedHoveredSlateSoundOverride {resourceObject=&Fortnite_UI_Ranked_Button_Hover_Cue}
    selectedPressedSlateSoundOverride {resourceObject=&Fortnite_UI_Ranked_Button_Back_Cue}
    triggeringInputAction {dataTable=/Game/Athena/UI/Frontend/Lobby/AthenaLobbyInputActionDataTable, rowName=HabaneroToggle}
   }
   slot {class=OverlaySlot, properties {padding=<ltrb 0 0 0 32>}}
  }
  Button_Close {
   widgetClassPath=&WBP_UIKit_Button_Quiet
   bInherited=true
   bIsVariable=true
   parent=Overlay_3
   widget=Button_Close
   properties {
    text=Back
    visibility=Visible
    hoveredSlateSoundOverride {resourceObject=&Fortnite_UI_Ranked_Button_Hover_Cue}
    pressedSlateSoundOverride {resourceObject=&Fortnite_UI_Ranked_Button_Back_Cue}
    selectedHoveredSlateSoundOverride {resourceObject=&Fortnite_UI_Ranked_Button_Hover_Cue}
    selectedPressedSlateSoundOverride {resourceObject=&Fortnite_UI_Ranked_Button_Back_Cue}
    triggeringInputAction {dataTable=&AthenaLobbyInputActionDataTable, rowName=HabaneroClose}
   }
   slot {class=OverlaySlot}
  }
  SafeZone_mobileclose {
   widgetClassPath=&SafeZone
   bIsVariable=true
   parent=Overlay_23
   widget=SafeZone_mobileclose
   properties {padBottom=false, padLeft=false}
   slot {class=OverlaySlot, properties {horizontalAlignment=HAlign_Right, padding=<ltrb 0 32 32 0>}}
  }
  VisibilityBorder_Touch {
   widgetClassPath=&CommonHardwareVisibilityBorder
   parent=SafeZone_mobileclose
   widget=VisibilityBorder_Touch
   properties {
    padding=<ltrb 0 32 0 0>
    visibilityQuery {
     autoDescription=" ANY( Input.Touch )"
     tokenStreamVersion=0
     userDescription=touch-only
     queryTokenStream [0,1,1,1,0]
     tagDictionary [{tagName=Input.Touch}]
    }
   }
   slot {class=SafeZoneSlot}
  }
  Button_CloseTouch {
   widgetClassPath=/FortUILibrary/Foundation/Buttons/CloseButton
   bInherited=true
   bIsVariable=true
   parent=VisibilityBorder_Touch
   widget=Button_CloseTouch
   properties {
    onButtonBaseClicked=(&self:WidgetTree.Button_CloseTouch.HandleOnButtonClicked)
    onButtonBaseDoubleClicked=(&self:WidgetTree.Button_CloseTouch.HandleOnButtonDoubleClicked)
    onButtonBaseHovered=(&self:WidgetTree.Button_CloseTouch.HandleOnButtonHovered)
    onButtonBaseUnhovered=(&self:WidgetTree.Button_CloseTouch.HandleOnButtonUnhovered)
    onSelectedChangedBase=(&self:WidgetTree.Button_CloseTouch.HandleOnSelectedChanged)
    hoveredSlateSoundOverride {resourceObject=&Fortnite_UI_Ranked_Button_Hover_Cue}
    pressedSlateSoundOverride {resourceObject=&Fortnite_UI_Ranked_Button_Back_Cue}
   }
   slot {class=BorderSlot, properties {padding=<ltrb 0 32 0 0>}}
  }
 ]
}
